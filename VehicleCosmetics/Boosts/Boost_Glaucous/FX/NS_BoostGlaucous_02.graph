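# Niagara system for the Glaucous vehicle boost cosmetic, drawing a mostly green-teal exhaust with energy, trail and flare at each boost socket.
# Short-lived mesh/sprite energy particles sit on sockets, while six-renderer ribbons trail behind with pulsing lightning color and width.
# An active GPU flare scales by camera alignment and uses occlusion sampling; a second is disabled.
# Painted-variant overrides recolor energy and flares; frontend-preview and game-mode parameters adjust values.
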NiagaraSystem /VehicleCosmetics/Boosts/Boost_Glaucous/FX/NS_BoostGlaucous_02
dependencies {
 usedDataInterfaces [
  &NiagaraDataInterfaceColorCurve
  &NiagaraDataInterfaceCurve
  /Script/Niagara.NiagaraDataInterfaceOcclusion
  &NiagaraDataInterfaceCamera
  &NiagaraDataInterfaceSkeletalMesh
  &NiagaraDataInterfaceArrayFloat3
 ]
 usedDynamicInputs [&SelectVectorFromArray]
 usedModules [
  &self:SystemSpawnScript.SystemScriptSource.SystemScriptGraph.NiagaraNodeAssignment_0.SetVariables_#1
  &self:SystemSpawnScript.SystemScriptSource.SystemScriptGraph.NiagaraNodeAssignment_1.SetVariables_#2
  &self:SystemSpawnScript.SystemScriptSource.SystemScriptGraph.NiagaraNodeAssignment_2.SetVariables_#3
  &SystemState
 ]
}
emitters {
 NE_VC_Boost_Energy {
  emitterUpdateScript [
   &EmitterState(Inactive Response=::NewEnumerator1, Life Cycle Mode=::NewEnumerator1, UseLoopDelay=false)
   &SpawnRate(SpawnRate=30, Use Spawn Probability=false) {module=SpawnRate002}
   &VC_FilteredSocketCount()
   &SpawnRate(SpawnGroup=1, SpawnRate=15, Use Spawn Probability=false) {enabled=false}
  ]
  particleSpawnScript [
   /SetVariables(Particles.VisibilityTag=$Emitter.SpawnGroup) {isSetParameters=true}
   &VC_PaintedVariantsOverride(
    Black=<rgba 0.007 0.007 0.007 1>
    BurntSienna=<rgba 0.55 0.1953 0.077 1>
    Cobalt=<rgba 0 0.055 1.1 1>
    Crimson=<rgba 1.6 0.024 0.0503 1>
    DefaultColor=<rgba 0 800 458.4494 1>
    ForestGreen=<rgba 0 1.7 0.0567 1>
    Grey=<rgba 0.18 0.18 0.18 1>
    Lime=<rgba 0.4007 0.7 0.035 1>
    Orange=<rgba 1 0.335 0.05 1>
    Pink=<rgba 0.6 0.075 0.6 1>
    Purple=<rgba 0.1275 0.0125 0.3 1>
    Saffron=<rgba 0.4 0.3684 0.0208 1>
    SkyBlue=<rgba 0 2.565 2.7 1>
    TitaniumWhite=<rgba 0.55 0.55 0.55 1>
   )
   &VC_PaintedVariantsOverride(
    Black=<rgba 0.007 0.007 0.007 1>
    BurntSienna=<rgba 1.2 0.3768 0.024 1>
    Cobalt=<rgba 0.17 0.2115 1 1>
    Crimson=<rgba 1.6 0.024 0.0503 1>
    DefaultColor=<rgba 20 6.5 2 1>
    ForestGreen=<rgba 0.2 1 0.2267 1>
    Grey=<rgba 0.08 0.08 0.08 1>
    Lime=<rgba 1.1 2 0 1>
    Orange=<rgba 2.3 0.3833 0 1>
    Pink=<rgba 2 0.3 2 1>
    Purple=<rgba 0.688 0.08 1.6 1>
    Saffron=<rgba 2 1.842 0.1042 1>
    SkyBlue=<rgba 0 2.85 3 1>
    TitaniumWhite=<rgba 2 2 2 1>
   ) {enabled=false, module=VC_PaintedVariantsOverride001}
   &InitializeParticle(
    Color=&MakeCustomLinearColorFromBool(Bool=$User.bIsFrontendPreview, True Color=<rgba 0 10 5.7306 1>, False Color=<rgba 0 10 5.7306 1>)
    Lifetime=0.05
    Mesh Scale Max=&MergeHelper_VectorSwitch(0=<3 6 3>, 1=<3.5 6 3.5>, 2=?, 3=?, Index=$Emitter.SpawnGroup, Num=::NewEnumerator0)
    Mesh Scale Min=&MergeHelper_VectorSwitch(0=<2.8 3 2.8>, 1=<2 2.25 2>, 2=?, 3=?, Index=$Emitter.SpawnGroup, Num=::NewEnumerator0)
    Mesh Scale Mode=::NewEnumerator2
    Position Mode=::NewEnumerator1
    Sprite Rotation Mode=::NewEnumerator2
    Sprite Size Mode=::NewEnumerator3
    UsePositionOffset=false
    Write Material Random=true
   )
   &PartitionParticles(Number Of Partitions=$Emitter.PortCount)
   &SkeletalMeshLocation(
    Bone / Socket Index=$Particles.PartitionParticles.Partition
    Bone / Socket Sampling Mode=::NewEnumerator5
    MaskSpawnGroup=false
    Mesh Sampling Type=::NewEnumerator6
    Sampled Position Offset=&AddVectorToPosition(Position=<0 0 0>, Vector=<0 0 0>)
    Skeletal Mesh=$System.SkeletalMesh
   )
   &AddVelocity(Velocity=<0 0 0>)
   &InitialMeshOrientation(
    Mesh Axis To Orient=<0 1 0>
    Mesh Orientation Mode=::NewEnumerator2
    Orientation Vector=<0 0 0>
    Rotation=&RandomRangeVector(Minimum=<0 -1 0>, Maximum=<0 1 0>)
    Use Rotation Vector=true
   )
   &OffsetPosition(Position Offset=<10 0 0>)
   /VehicleCosmetics/FX/VC_FXParents/Modules/DLR_ProductPreviews(GameModeID:NiagaraInt32=$System.GameModeID, bPreviewVehicleSpeed:NiagaraBool=$System.bPreviewVehicleSpeed)
  ]
  particleUpdateScript [
   &ParticleState()
   &InheritSourceMovement(
    Applied Position Scale=&Multiply_VectorByFloat(
     Vector=<1 1 1>
     Float=&FloatFromCurve(FloatCurve=DataInterface() {propertyValues {Curve {
         defaultValue=1
         keys [
          {arriveTangent=-1.7885, interpMode=RCIM_Cubic, leaveTangent=-1.7885, tangentMode=RCTM_User, time=0.05, value=1}
          {interpMode=RCIM_Cubic, tangentMode=RCTM_User, time=1, value=0}
         ]
        }}})
    )
    ApplyForce=false
    ApplyPosition=true
    ApplyPositionFalloff=false
   ) {enabled=false, module=InheritSourceMovement001}
   &SolveForcesAndVelocity(Clamp Velocity=false, Limit Acceleration=false)
  ]
  rendererClasses [&NiagaraMeshRendererProperties, &NiagaraSpriteRendererProperties]
  renderers [
   {rendererClass=&NiagaraMeshRendererProperties, rendererIndex=0}
   {rendererClass=&NiagaraSpriteRendererProperties, rendererIndex=1}
  ]
 }
 NE_VC_Boost_Ribbon_Parent {
  emitterSpawnScript [/SetVariables(Emitter.MaskShifter:LinearColor=&SelectLinearColorFromArray(
     Array Sampling Mode=::NewEnumerator1
     Direct Array Index=0
     Color Selection Array=DataInterface() {propertyValues {ColorData [<rgba 1 1 1 1>, <rgba 0 1 0 1>, <rgba 0 0 1 1>]}}
     Random Seed=?
    )) {isSetParameters=true}]
  emitterUpdateScript [
   &EmitterState()
   &SpawnRate(SpawnGroup=1, SpawnRate=30, Use Spawn Probability=false) {enabled=false}
   &VC_FilteredSocketCount()
   &VC_RateQualityHandler()
   &VC_SpawnRibbon_v1(
    Number of Ribbons=$Emitter.PortCount
    SpawnRate=&MakeCustomFloatFromBool(Bool=$User.bIsFrontendPreview, True Float=60, False Float=60)
   )
  ]
  particleSpawnScript [
   &VC_PaintedVariantsOverride() {enabled=false}
   &VC_InitializeRibbon_v1(
    Enable Skeletal Position=true
    Lifetime=&MakeCustomFloatFromBool(Bool=$User.bIsFrontendPreview, True Float=3, False Float=1.2)
    Ribbon Width=200
   )
   /SetVariables(Particles.SecondaryColor:LinearColor=&MakeCustomLinearColorFromBool(Bool=$User.bIsFrontendPreview, True Color=<rgba 1 0.7152 0 1>, False Color=<rgba 1 0.7405 0.0269 1>)) {enabled=false, isSetParameters=true}
   &AddVelocity(Velocity=&MakeCustomVectorFromBool(Bool=$User.bIsFrontendPreview, True Vector=<-2000 0 0>, False Vector=<-1200 0 0>)) {module=AddVelocity001}
   /SetVariables(Particles.SecondaryVisTag:NiagaraInt32=&MakeCustomIntFromBool(Bool=$User.bIsFrontendPreview, True INT=1, False INT=1)) {enabled=false, isSetParameters=true}
   &DLR_ProductPreviews(GameModeID:NiagaraInt32=$System.GameModeID, bPreviewVehicleSpeed:NiagaraBool=$System.bPreviewVehicleSpeed)
   /SetVariables(
    Particles.lightningcolor1:LinearColor=&Multiply_LinearColorByFloat(
     Linear Color=&MakeLinearColorFromVectorAndFloat(
      "Vector (RGB)"=&MakeVectorFromLinearColorRGB(Color=&ColorFromCurve(ColorCurve=DataInterface() {propertyValues {
          AlphaCurve {keys [<at 0 1 RCIM_Cubic>,<at 1 0 RCIM_Cubic>]}
          BlueCurve {keys [<at 0 13.8635 RCIM_Cubic>,<at 0.5617 0.1848>,<at 1 0 RCIM_Cubic>]}
          GreenCurve {keys [<at 0 15 RCIM_Cubic>,<at 0.5617 0.2>,<at 1 0 RCIM_Cubic>]}
          RedCurve {keys [<at 0 0.75 RCIM_Cubic>,<at 0.5617 0.01>,<at 1 0 RCIM_Cubic>]}
         }}))
      "Float (Alpha)"=&FloatFromCurve(FloatCurve=DataInterface() {propertyValues {Curve {keys [<at 0 0>,<at 0.1 0.5>,<at 0.95 0.5>,<at 1 0>]}}})
     )
     Float=&Cosine(Normalized Angle=$Emitter.Age, Period=0.4, Scale=15, Bias=11)
    )
    Particles.lightningcolor2:LinearColor=&Multiply_LinearColorByFloat(
     Linear Color=&MakeLinearColorFromVectorAndFloat(
      "Vector (RGB)"=&MakeVectorFromLinearColorRGB(Color=&ColorFromCurve(ColorCurve=DataInterface() {propertyValues {
          AlphaCurve {keys [<at 0 1 RCIM_Cubic>,<at 1 0 RCIM_Cubic>]}
          BlueCurve {keys [<at 0 0.9242 RCIM_Cubic>,<at 0.4015 0.1543>,<at 1 0 RCIM_Cubic>]}
          GreenCurve {keys [<at 0 1 RCIM_Cubic>,<at 0.4015 0.1669>,<at 1 0 RCIM_Cubic>]}
          RedCurve {keys [<at 0 0.05 RCIM_Cubic>,<at 0.4015 0.0083>,<at 1 0 RCIM_Cubic>]}
         }}))
      "Float (Alpha)"=&FloatFromCurve(FloatCurve=DataInterface() {propertyValues {Curve {keys [<at 0 0>,<at 0.1 0.5>,<at 0.7 0.5>,<at 0.9 0>]}}})
     )
     Float=&Sine(Normalized Angle=$Emitter.Age, Period=0.6, Scale=20, Bias=10)
    )
    Particles.lightningcolor3:LinearColor=&Multiply_LinearColorByFloat(
     Linear Color=&MakeLinearColorFromVectorAndFloat(
      "Vector (RGB)"=&MakeVectorFromLinearColorRGB(Color=&ColorFromCurve(ColorCurve=DataInterface() {propertyValues {
          AlphaCurve {keys [<at 0 1 RCIM_Cubic>,<at 1 0 RCIM_Cubic>]}
          BlueCurve {keys [<at 0 0.9242 RCIM_Cubic>,<at 0.3992 0.0747>,<at 1 0 RCIM_Cubic>]}
          GreenCurve {keys [<at 0 1 RCIM_Cubic>,<at 0.3992 0.0808>,<at 1 0 RCIM_Cubic>]}
          RedCurve {keys [<at 0 0.05 RCIM_Cubic>,<at 0.3992 0.004>,<at 1 0 RCIM_Cubic>]}
         }}))
      "Float (Alpha)"=&FloatFromCurve(FloatCurve=DataInterface() {propertyValues {Curve {keys [<at 0 0>,<at 0.1 0.8>,<at 0.7 0.8>,<at 0.9 0>]}}})
     )
     Float=&Cosine(Normalized Angle=$Emitter.Age, Period=0.7, Scale=50, Bias=7)
    )
    Particles.lightningcolor4:LinearColor=&Multiply_LinearColorByFloat(
     Linear Color=&MakeLinearColorFromVectorAndFloat(
      "Vector (RGB)"=&MakeVectorFromLinearColorRGB(Color=&ColorFromCurve(ColorCurve=DataInterface() {propertyValues {
          AlphaCurve {keys [<at 0 1 RCIM_Cubic>,<at 1 0 RCIM_Cubic>]}
          BlueCurve {keys [<at 0 0.9242 RCIM_Cubic>,<at 0.4769 0.1117>,<at 1 0 RCIM_Cubic>]}
          GreenCurve {keys [<at 0 1 RCIM_Cubic>,<at 0.4769 0.1208>,<at 1 0 RCIM_Cubic>]}
          RedCurve {keys [<at 0 0.05 RCIM_Cubic>,<at 0.4769 0.006>,<at 1 0 RCIM_Cubic>]}
         }}))
      "Float (Alpha)"=&FloatFromCurve(FloatCurve=DataInterface() {propertyValues {Curve {keys [<at 0 0>,<at 0.1 0.5>,<at 0.7 0.5>,<at 0.9 0>]}}})
     )
     Float=&Cosine(Normalized Angle=$Emitter.Age, Period=0.4, Scale=15, Bias=11)
    )
   ) {isSetParameters=true}
   /SetVariables(Particles.lightningwidth1:NiagaraFloat=&Multiply_Float(
     A=&FloatFromCurve(
      FloatCurve=DataInterface() {propertyValues {Curve {
         defaultValue=1
         keys [
          {arriveTangent=0.1434, interpMode=RCIM_Cubic, leaveTangent=0.1434, tangentMode=RCTM_User, time=0, value=0.1}
          {arriveTangent=4.2631, interpMode=RCIM_Cubic, leaveTangent=4.2631, tangentMode=RCTM_User, time=0.5, value=1.5}
          {arriveTangent=-4.4621, interpMode=RCIM_Cubic, leaveTangent=-4.4621, tangentMode=RCTM_User, time=1, value=0.4}
         ]
        }}}
      Scale Curve=5
     )
     B=&Cosine(Normalized Angle=$Emitter.Age, Period=1, Scale=&RandomRangeFloat(Minimum=300, Maximum=600), Bias=1)
    )) {isSetParameters=true}
   &Color(Color=&ColorFromCurve(ColorCurve=DataInterface() {propertyValues {
       AlphaCurve {keys [<at 0 1 RCIM_Cubic>,<at 0.5539 0.042>,<at 1 0 RCIM_Cubic>]}
       BlueCurve {keys [<at 0 1 RCIM_Cubic>,<at 1 1 RCIM_Cubic>]}
       GreenCurve {keys [<at 0 1 RCIM_Cubic>,<at 1 1 RCIM_Cubic>]}
       RedCurve {keys [<at 0 1 RCIM_Cubic>,<at 1 1 RCIM_Cubic>]}
      }}))
  ]
  particleUpdateScript [
   &ParticleState()
   &InheritSourceMovement(
    Applied Position Scale=&Multiply_VectorByFloat(
     Vector=<1 1 1>
     Float=&FloatFromCurve(FloatCurve=DataInterface() {propertyValues {Curve {
         defaultValue=1
         keys [
          {arriveTangent=-15.5437, interpMode=RCIM_Cubic, leaveTangent=-15.5437, tangentMode=RCTM_User, time=0, value=1}
          {arriveTangent=-0.6523, interpMode=RCIM_Cubic, leaveTangent=-0.6523, tangentMode=RCTM_User, time=0.1, value=0.3}
          {arriveTangent=-0.1632, interpMode=RCIM_Cubic, leaveTangent=-0.1632, tangentMode=RCTM_User, time=1, value=0}
         ]
        }}})
    )
    ApplyForce=false
    ApplyPosition=true
    ApplyPositionFalloff=false
   )
   &SolveForcesAndVelocity(Clamp Velocity=false, Limit Acceleration=false)
   &VC_UpdateRibbon_v1(Enable Skeletal Position=true)
   /SetVariables(
    Particles.Base:LinearColor=&MakeCustomLinearColorFromBool(
     Bool=$User.bIsFrontendPreview
     True Color=&ColorFromCurve(ColorCurve=DataInterface() {propertyValues {
        AlphaCurve {keys [<at 0 0.4 RCIM_Cubic>,<at 0.7484 0.1>,<at 1 0 RCIM_Cubic>]}
        BlueCurve {keys [<at 0.1568 0.224 RCIM_Cubic>,<at 0.3214 0.027>,<at 0.4593 0.0007 RCIM_Cubic>]}
        GreenCurve {keys [<at 0.1568 0.8 RCIM_Cubic>,<at 0.3214 0.1029>,<at 0.4593 0.01 RCIM_Cubic>]}
        RedCurve {keys [<at 0.1568 0 RCIM_Cubic>,<at 0.3214 0.0017>,<at 0.4593 0.002 RCIM_Cubic>]}
       }})
     False Color=&ColorFromCurve(ColorCurve=DataInterface() {propertyValues {
        AlphaCurve {keys [<at 0 0.4 RCIM_Cubic>,<at 0.7484 0.1>,<at 1 0 RCIM_Cubic>]}
        BlueCurve {keys [<at 0.1568 0.224 RCIM_Cubic>,<at 0.3214 0.027>,<at 0.9419 0.0007 RCIM_Cubic>]}
        GreenCurve {keys [<at 0.1568 0.8 RCIM_Cubic>,<at 0.3214 0.1029>,<at 0.9419 0.01 RCIM_Cubic>]}
        RedCurve {keys [<at 0.1568 0 RCIM_Cubic>,<at 0.3214 0.0017>,<at 0.9419 0.002 RCIM_Cubic>]}
       }})
    )
    Particles.PatterBase:LinearColor=&MergeHelper_ColorSwitch(
     0=&ColorFromCurve(ColorCurve=DataInterface() {propertyValues {
        AlphaCurve {keys [<at 0 1 RCIM_Cubic>,<at 1 0 RCIM_Cubic>]}
        BlueCurve {keys [<at 0 0.056 RCIM_Cubic>]}
        GreenCurve {keys [<at 0 0.2 RCIM_Cubic>]}
        RedCurve {keys [<at 0 0 RCIM_Cubic>]}
       }})
     1=&ColorFromCurve(ColorCurve=DataInterface() {propertyValues {
        AlphaCurve {keys [<at 0 1 RCIM_Cubic>,<at 1 0 RCIM_Cubic>]}
        BlueCurve {keys [<at 0 0.056 RCIM_Cubic>]}
        GreenCurve {keys [<at 0 0.2 RCIM_Cubic>]}
        RedCurve {keys [<at 0 0 RCIM_Cubic>]}
       }})
     2=&ColorFromCurve(ColorCurve=DataInterface() {propertyValues {
        AlphaCurve {keys [<at 0 1 RCIM_Cubic>,<at 1 0 RCIM_Cubic>]}
        BlueCurve {keys [<at 0 0.056 RCIM_Cubic>]}
        GreenCurve {keys [<at 0 0.2 RCIM_Cubic>]}
        RedCurve {keys [<at 0 0 RCIM_Cubic>]}
       }})
     3=?
     Index=$User.GameModeID
     Num=::NewEnumerator1
    )
    Particles.Pattern02:LinearColor=&MakeCustomLinearColorFromBool(
     Bool=$User.bIsFrontendPreview
     True Color=&ColorFromCurve(ColorCurve=DataInterface() {propertyValues {
        AlphaCurve {keys [<at 0 1 RCIM_Cubic>,<at 0.4615 0.0272>,<at 0.675 0 RCIM_Cubic>]}
        BlueCurve {keys [<at 0 0.372 RCIM_Cubic>,<at 0.5104 0.05>,<at 0.9442 0>]}
        GreenCurve {keys [<at 0 0.8 RCIM_Cubic>,<at 0.5104 0.0387>,<at 0.9442 0>]}
        RedCurve {keys [<at 0 0 RCIM_Cubic>,<at 0.5104 0.0207>,<at 0.9442 0>]}
       }})
     False Color=&ColorFromCurve(ColorCurve=DataInterface() {propertyValues {
        AlphaCurve {keys [<at 0 1 RCIM_Cubic>,<at 0.4615 0.0272>,<at 0.675 0 RCIM_Cubic>]}
        BlueCurve {keys [<at 0 0.372 RCIM_Cubic>,<at 0.8129 0.05>,<at 0.9642 0>]}
        GreenCurve {keys [<at 0 0.8 RCIM_Cubic>,<at 0.8129 0.0387>,<at 0.9642 0>]}
        RedCurve {keys [<at 0 0 RCIM_Cubic>,<at 0.8129 0.0207>,<at 0.9642 0>]}
       }})
    )
    Particles.PatternHightlight:LinearColor=&MakeCustomLinearColorFromBool(
     Bool=$User.bIsFrontendPreview
     True Color=&ColorFromCurve(ColorCurve=DataInterface() {propertyValues {
        AlphaCurve {keys [<at 0 1 RCIM_Cubic>,<at 0.1279 0.1>,<at 0.3103 0 RCIM_Cubic>]}
        BlueCurve {keys [<at 0 0.6 RCIM_Cubic>]}
        GreenCurve {keys [<at 0 2.322 RCIM_Cubic>]}
        RedCurve {keys [<at 0 3 RCIM_Cubic>]}
       }})
     False Color=&ColorFromCurve(ColorCurve=DataInterface() {propertyValues {
        AlphaCurve {keys [<at 0 1 RCIM_Cubic>,<at 0.3548 0.1>,<at 0.8018 0 RCIM_Cubic>]}
        BlueCurve {keys [<at 0 0.3 RCIM_Cubic>]}
        GreenCurve {keys [<at 0 2.2372 RCIM_Cubic>]}
        RedCurve {keys [<at 0 3 RCIM_Cubic>]}
       }})
    )
    Particles.TrusterCOlor02:LinearColor=&MakeCustomLinearColorFromBool(
     Bool=$User.bIsFrontendPreview
     True Color=&ColorFromCurve(ColorCurve=DataInterface() {propertyValues {
        AlphaCurve {keys [<at 0.3214 1 RCIM_Cubic>,<at 0.6506 0.1>,<at 1 0 RCIM_Cubic>]}
        BlueCurve {keys [<at 0.3258 5.6005 RCIM_Cubic>,<at 1 0 RCIM_Cubic>]}
        GreenCurve {keys [<at 0.3258 20 RCIM_Cubic>,<at 1 0 RCIM_Cubic>]}
        RedCurve {keys [<at 0.3258 0 RCIM_Cubic>,<at 1 0 RCIM_Cubic>]}
       }})
     False Color=&ColorFromCurve(ColorCurve=DataInterface() {propertyValues {
        AlphaCurve {keys [<at 0.3214 1 RCIM_Cubic>,<at 0.6506 0.1>,<at 1 0 RCIM_Cubic>]}
        BlueCurve {keys [<at 0.3214 5.6005 RCIM_Cubic>,<at 1 0 RCIM_Cubic>]}
        GreenCurve {keys [<at 0.3214 20 RCIM_Cubic>,<at 1 0 RCIM_Cubic>]}
        RedCurve {keys [<at 0.3214 0 RCIM_Cubic>,<at 1 0 RCIM_Cubic>]}
       }})
    )
   ) {isSetParameters=true}
   &ScaleRibbonWidth(Ribbon Width Scale=&FloatFromCurve(FloatCurve=DataInterface() {propertyValues {Curve {keys [<at 0 1>,<at 1 0>]}}}))
   /SetVariables(Particles.Lightning:LinearColor=&RandomRangeLinearColor(
     Minimum=&ColorFromCurve(ColorCurve=DataInterface() {propertyValues {
        AlphaCurve {keys [<at 0 1 RCIM_Cubic>,<at 0.3636 0 RCIM_Cubic>]}
        BlueCurve {keys [<at 0 69.5639 RCIM_Cubic>,<at 0.4184 0.3478 RCIM_Cubic>]}
        GreenCurve {keys [<at 0 200 RCIM_Cubic>,<at 0.4184 1 RCIM_Cubic>]}
        RedCurve {keys [<at 0 0 RCIM_Cubic>,<at 0.4184 0 RCIM_Cubic>]}
       }})
     Maximum=&ColorFromCurve(ColorCurve=DataInterface() {propertyValues {
        AlphaCurve {keys [<at 0 1 RCIM_Cubic>,<at 0.638 0 RCIM_Cubic>]}
        BlueCurve {keys [<at 0 0.3478 RCIM_Cubic>,<at 1 0 RCIM_Cubic>]}
        GreenCurve {keys [<at 0 1 RCIM_Cubic>,<at 1 0 RCIM_Cubic>]}
        RedCurve {keys [<at 0 0 RCIM_Cubic>,<at 1 0 RCIM_Cubic>]}
       }})
    )) {isSetParameters=true}
  ]
  rendererClasses [&NiagaraRibbonRendererProperties]
  renderers [
   {rendererClass=&NiagaraRibbonRendererProperties}
   {rendererClass=&NiagaraRibbonRendererProperties, rendererIndex=1}
   {rendererClass=&NiagaraRibbonRendererProperties, rendererIndex=2}
   {rendererClass=&NiagaraRibbonRendererProperties, rendererIndex=3}
   {rendererClass=&NiagaraRibbonRendererProperties, rendererIndex=4}
   {rendererClass=&NiagaraRibbonRendererProperties, rendererIndex=5}
  ]
 }
 NE_VC_Flare {
  bEnabled=false
  simTarget=GPUComputeSim
  emitterUpdateScript [
   &EmitterState(Inactive Response=::NewEnumerator1, Life Cycle Mode=::NewEnumerator1, Loop Duration=1, UseLoopDelay=false)
   &VC_FilteredSocketCount()
   &SpawnRate(SpawnRate=6, Use Spawn Probability=false)
  ]
  particleSpawnScript [
   &VC_PaintedVariantsOverride(
    Black=<rgba 0.0625 0.0625 0.0625 1>
    BurntSienna=<rgba 0.1771 0.068 0.0387 1>
    Cobalt=<rgba 0.0479 0.0618 0.799 1>
    Crimson=<rgba 0.799 0.0479 0.0683 1>
    DefaultColor=<rgba 0.08 0.8 0.4017 1>
    ForestGreen=<rgba 0.0495 0.3391 0.0641 1>
    Grey=<rgba 0.2087 0.2087 0.2087 1>
    Lime=<rgba 0.4235 0.799 0.0479 1>
    Orange=<rgba 0.799 0.3099 0.0479 1>
    Pink=<rgba 0.799 0.0479 0.5259 1>
    Purple=<rgba 0.3443 0.0278 0.799 1>
    Saffron=<rgba 0.799 0.6304 0.0764 1>
    SkyBlue=<rgba 0.0479 0.5477 0.799 1>
    TitaniumWhite=<rgba 0.3698 0.3698 0.3698 1>
   )
   &InitializeParticle(
    Color=&MakeCustomLinearColorFromBool(Bool=$User.bIsFrontendPreview, True Color=<rgba 0 1 0.4772 1>, False Color=<rgba 0 1 0.5731 1>)
    Lifetime=&MergeHelper_FloatSwitch(0=0.5, 1=0.5, 2=0.5, 3=?, Index=$User.GameModeID, Num=::NewEnumerator1)
    Sprite Size=&RandomRangeVector2D(Minimum=<800 400>, Maximum=<1400 900>)
    Sprite Size Mode=::NewEnumerator4
    UsePositionOffset=false
    Write Material Random=true
   )
   &PartitionParticles(Number Of Partitions=$Emitter.PortCount)
   &SkeletalMeshLocation(
    Bone / Socket Index=$Particles.PartitionParticles.Partition
    Bone / Socket Sampling Mode=::NewEnumerator5
    MaskSpawnGroup=false
    Mesh Sampling Type=::NewEnumerator6
    Sampled Position Offset=&AddVectorToPosition(Position=$User.DTCarOffset002, Vector=<-20 0 0>)
    Skeletal Mesh=$System.SkeletalMesh
   )
   &AddVelocity(Velocity=<-1500 0 0>) {enabled=false}
   /SetVariables(Particles.SecundaryColor:LinearColor=<rgba 0 0.3218 0.55 1>, Particles.SpriteSizeSecunday:Vector2f=<215 215>) {isSetParameters=true, module=SetVariables_#4}
   &CameraOffset(Camera Offset Amount=50)
  ]
  particleUpdateScript [
   &ParticleState()
   /SetVariables(
    Particles.Streak_01_Offset:NiagaraPosition=&AddVectorToPosition(Position=$Particles.SkeletalMeshLocation.SampledPosition, Vector=<0 0 0>)
    Particles.Streak_01_Size:Vector2f=<350 25>
   ) {isSetParameters=true, module=SetVariables_#5}
   &self:ScratchModule_04(
    NumberOfRings=5
    OcclusionSize=4.5
    Position:NiagaraPosition=&TransformPosition(Position=$Particles.Streak_01_Offset, Source Space=::Local, Destination Space=::World)
    SamplesPerRing=5
    SystemX Sample Offset=5
   ) {module=ScratchModule_04}
   /SetVariables(Particles.CameraScale:NiagaraFloat=&MakeFloatFromVector(
     Vector=&Multiply_VectorByFloat(
      Vector=&ReturnCameraVector(Camera Vectors=::NewEnumerator4, Coordinate Space=::Simulation, Origin Position=$Particles.Position)
      Float=&DotProduct(
       Vector A=<1 0 -1.5>
       Vector B=&ReturnCameraVector(Camera Vectors=::NewEnumerator4, Coordinate Space=::Simulation, Origin Position=$Particles.Position)
      )
     )
     Channel=::NewEnumerator0
    )) {isSetParameters=true, module=SetVariables_#6}
   /SetVariables(Particles.CameraScale:NiagaraFloat=&Multiply_Float(
     A=$Particles.CameraScale
     B=&RemapRange(
      Input Value To Remap=&DotProduct(
       Vector A=&NormalizeVector(
        Vector To Normalize=&RotateVectorByQuaternion(Vector To Rotate=<1 0 0>, Quaternion=$StackContext.SkeletalMeshLocation.SampledRotation)
        Safe Normalize=true
        Safe Fallback Threshold=0
        Safe Fallback Vector=<0 0 0.001>
       )
       Vector B=&NormalizeVector(
        Vector To Normalize=&ReturnCameraVector(Camera Vectors=::NewEnumerator1, Coordinate Space=::Simulation)
        Safe Normalize=true
        Safe Fallback Threshold=0
        Safe Fallback Vector=<0 0 0.001>
       )
      )
      Input Min=0
      Input Max=-1
      Output Min=0
      Output Max=1
     )
    )) {isSetParameters=true, module=SetVariables_#7}
   &ScaleColor(Scale Alpha=$Particles.Visibility, Scale RGB=&VectorFromFloat(Value=2), ScaleA=true, ScaleRGB=true) {module=ScaleColor002}
   &SolveForcesAndVelocity(Clamp Velocity=false, Limit Acceleration=false) {enabled=false}
   &ScaleColor(
    Linear Color Curve=DataInterface() {propertyValues {
      AlphaCurve {keys [<at 0 1>,<at 0.1069 0.2>,<at 0.2416 1>,<at 0.4086 0.2>,<at 0.637 1>,<at 0.8303 0.2>]}
      BlueCurve {keys [<at 0 1>,<at 1 1>]}
      GreenCurve {keys [<at 0 1>,<at 1 1>]}
      RedCurve {keys [<at 0 1>,<at 1 1>]}
     }}
    Scale Mode=2
   ) {enabled=false}
  ]
  rendererClasses [&NiagaraSpriteRendererProperties]
  renderers [
   {rendererClass=&NiagaraSpriteRendererProperties}
   {rendererClass=&NiagaraSpriteRendererProperties, rendererIndex=1}
  ]
 }
 NE_VC_Flare001 {
  simTarget=GPUComputeSim
  emitterUpdateScript [
   &EmitterState(Inactive Response=::NewEnumerator1, Life Cycle Mode=::NewEnumerator1, Loop Duration=1, UseLoopDelay=false)
   &VC_FilteredSocketCount()
   &SpawnRate(SpawnRate=6, Use Spawn Probability=false)
  ]
  particleSpawnScript [
   &VC_PaintedVariantsOverride(
    Black=<rgba 0.0625 0.0625 0.0625 1>
    BurntSienna=<rgba 0.1771 0.068 0.0387 1>
    Cobalt=<rgba 0.0479 0.0618 0.799 1>
    Crimson=<rgba 0.799 0.0479 0.0683 1>
    DefaultColor=<rgba 0.08 0.8 0.4017 1>
    ForestGreen=<rgba 0.0495 0.3391 0.0641 1>
    Grey=<rgba 0.2087 0.2087 0.2087 1>
    Lime=<rgba 0.4235 0.799 0.0479 1>
    Orange=<rgba 0.799 0.3099 0.0479 1>
    Pink=<rgba 0.799 0.0479 0.5259 1>
    Purple=<rgba 0.3443 0.0278 0.799 1>
    Saffron=<rgba 0.799 0.6304 0.0764 1>
    SkyBlue=<rgba 0.0479 0.5477 0.799 1>
    TitaniumWhite=<rgba 0.3698 0.3698 0.3698 1>
   )
   &InitializeParticle(
    Color=&MakeCustomLinearColorFromBool(Bool=$User.bIsFrontendPreview, True Color=<rgba 0 0.8 0.3818 1>, False Color=<rgba 0 0.8 0.4584 1>)
    Lifetime=&MergeHelper_FloatSwitch(0=0.5, 1=0.5, 2=0.5, 3=?, Index=$User.GameModeID, Num=::NewEnumerator1)
    Position Mode=::NewEnumerator1
    Sprite Size=&RandomRangeVector2D(Minimum=<1500 600>, Maximum=<1500 600>)
    Sprite Size Mode=::NewEnumerator4
    UsePositionOffset=false
    Write Material Random=true
   )
   &PartitionParticles(Number Of Partitions=$Emitter.PortCount, Particles Per Partition=1, Partition Mode:ENiagara_PartitionMode_v2=::NewEnumerator1)
   &SkeletalMeshLocation(
    Bone / Socket Index=$Particles.PartitionParticles.Partition
    Bone / Socket Sampling Mode=::NewEnumerator5
    MaskSpawnGroup=false
    Mesh Sampling Type=::NewEnumerator6
    Sampled Position Offset=&AddVectorToPosition(Position=$User.DTCarOffset002, Vector=<-20 0 0>)
    Skeletal Mesh=$System.SkeletalMesh
   )
   &AddVelocity(Velocity=&VC_Mask_Vectors(Coro Value:Vector3f=<0 0 0>, FrontEnd Value:Vector3f=<0 0 0>, FN Value:Vector3f=<-1200 0 0>))
   /SetVariables(Particles.RadialGlowSize:Vector2f=&Multiply_Vector2D(A=<0.1 0.2>, B=$Particles.SpriteSize)) {isSetParameters=true, module=SetVariables_#4}
   &DLR_ProductPreviews(GameModeID:NiagaraInt32=$System.GameModeID, bPreviewVehicleSpeed:NiagaraBool=$System.bPreviewVehicleSpeed)
  ]
  particleUpdateScript [
   &ParticleState()
   &CameraOffset(Camera Offset Amount=70)
   /SetVariables(Particles.Streak_01_Offset:NiagaraPosition=$Particles.SkeletalMeshLocation.SampledPosition) {isSetParameters=true, module=SetVariables_#5}
   &self:ScratchModule_06(
    NumberOfRings=5
    OcclusionSize=4.5
    Position:NiagaraPosition=&TransformPosition(Position=$Particles.SkeletalMeshLocation.SampledPosition, Source Space=::Local, Destination Space=::World)
    SamplesPerRing=5
    SystemX Sample Offset=5
   ) {module=ScratchModule_06}
   /SetVariables(Particles.CameraScale:NiagaraFloat=&MakeFloatFromVector(
     Vector=&Multiply_VectorByFloat(
      Vector=&ReturnCameraVector(Camera Vectors=::NewEnumerator4, Coordinate Space=::Simulation, Origin Position=$Particles.Position)
      Float=&DotProduct(
       Vector A=<1 0 -1.5>
       Vector B=&ReturnCameraVector(Camera Vectors=::NewEnumerator4, Coordinate Space=::Simulation, Origin Position=$Particles.Position)
      )
     )
     Channel=::NewEnumerator0
    )) {isSetParameters=true, module=SetVariables_#6}
   /SetVariables(Particles.CameraScale:NiagaraFloat=&Multiply_Float(
     A=$Particles.CameraScale
     B=&RemapRange(
      Input Value To Remap=&DotProduct(
       Vector A=&NormalizeVector(
        Vector To Normalize=&RotateVectorByQuaternion(Vector To Rotate=<1 0 0>, Quaternion=$StackContext.SkeletalMeshLocation.SampledRotation)
        Safe Normalize=true
        Safe Fallback Threshold=0
        Safe Fallback Vector=<0 0 0.001>
       )
       Vector B=&NormalizeVector(
        Vector To Normalize=&ReturnCameraVector(Camera Vectors=::NewEnumerator1, Coordinate Space=::Simulation)
        Safe Normalize=true
        Safe Fallback Threshold=0
        Safe Fallback Vector=<0 0 0.001>
       )
      )
      Input Min=0
      Input Max=-1
      Output Min=0
      Output Max=1
     )
    )) {isSetParameters=true, module=SetVariables_#7}
   &ScaleSpriteSize(
    Uniform Curve Index=$Particles.CameraScale
    Uniform Curve Sprite Scale=DataInterface() {propertyValues {Curve {keys [
        {arriveTangent=5.9531, interpMode=RCIM_Cubic, leaveTangent=5.9531, tangentMode=RCTM_User, time=0, value=0.5}
        {arriveTangent=0.0477, interpMode=RCIM_Cubic, leaveTangent=0.0477, tangentMode=RCTM_User, time=0.1, value=0.95}
        {tangentMode=RCTM_User, time=1, value=1}
       ]}}}
   )
   &ScaleColor(Scale Alpha=$Particles.Visibility, Scale RGB=&VectorFromFloat(Value=2), ScaleA=true, ScaleRGB=true) {module=ScaleColor002}
   /SetVariables(Particles.RadialGlowColor:LinearColor=&LerpLinearColors(StartColor=<rgba 0.0349 0.5123 0.9583 0>, EndColor=<rgba 0.0365 0.5346 1 1>, LerpFactor=$Particles.Visibility)) {isSetParameters=true}
   &InheritSourceMovement(
    Applied Position Scale=&VectorFromFloat(Value=&FloatFromCurve(FloatCurve=DataInterface() {propertyValues {Curve {
         defaultValue=1
         keys [
          {arriveTangent=-1.7885, interpMode=RCIM_Cubic, leaveTangent=-1.7885, tangentMode=RCTM_User, time=0.05, value=1}
          {interpMode=RCIM_Cubic, tangentMode=RCTM_User, time=1, value=0}
         ]
        }}}))
    ApplyForce=false
    ApplyPosition=true
    ApplyPositionFalloff=false
   ) {enabled=false}
   &ScaleColor(
    Linear Color Curve=DataInterface() {propertyValues {
      AlphaCurve {keys [<at 0 1>,<at 0.1069 0.2>,<at 0.2416 1>,<at 0.4086 0.2>,<at 0.637 1>,<at 0.8303 0.2>]}
      BlueCurve {keys [<at 0 1>,<at 1 1>]}
      GreenCurve {keys [<at 0 1>,<at 1 1>]}
      RedCurve {keys [<at 0 1>,<at 1 1>]}
     }}
    Scale Mode=2
   ) {enabled=false}
   &SolveForcesAndVelocity(Clamp Velocity=false, Limit Acceleration=false)
  ]
  rendererClasses [&NiagaraSpriteRendererProperties]
  renderers [
   {rendererClass=&NiagaraSpriteRendererProperties}
   {rendererClass=&NiagaraSpriteRendererProperties, rendererIndex=1}
  ]
 }
}
summary {
 systemName=NS_BoostGlaucous_02
 emitters [
  {emitterName=NE_VC_Boost_Ribbon_Parent, rendererClasses [&NiagaraRibbonRendererProperties]}
  {
   emitterName=NE_VC_Boost_Energy
   rendererClasses [&NiagaraMeshRendererProperties, &NiagaraSpriteRendererProperties]
  }
  {bEnabled=false, emitterName=NE_VC_Flare, simTarget=GPUComputeSim, rendererClasses [&NiagaraSpriteRendererProperties]}
  {emitterName=NE_VC_Flare001, simTarget=GPUComputeSim, rendererClasses [&NiagaraSpriteRendererProperties]}
 ]
 userVariables [
  User.bIsFrontendPreview {type=NiagaraBool}
  User.bIsPainted {type=NiagaraBool}
  User.blsIsCoroColor {type=NiagaraBool}
  User.bPreviewVehicleSpeed {type=NiagaraBool}
  User.ColorVariantID {type=NiagaraInt32}
  User.DTCarOffset {type=NiagaraPosition, description="", defaultValue {}}
  User.DTCarOffset {type=Vector3f, description="", defaultValue {struct=Vector3f, value=<0 0 0>}}
  User.DTCarOffset001 {type=Vector3f, description="", defaultValue {struct=Vector3f, value=<0 0 0>}}
  User.DTCarOffset002 {type=NiagaraPosition, description="", defaultValue {}}
  User.FlareColorArray {
   type=NiagaraDataInterfaceArrayColor
   description=""
   defaultValue {
    struct=NiagaraExt_VariableValue_DataInterface
    value {
     dataInterface=&self:NiagaraDataInterfaceArrayColor_0
     dataInterfaceClass=&NiagaraDataInterfaceArrayColor
    }
   }
  }
  User.FlareScale {type=NiagaraFloat, defaultValue=1, description=""}
  User.FlareScaleArray {
   type=NiagaraDataInterfaceArrayFloat
   defaultValue {
    struct=NiagaraExt_VariableValue_DataInterface
    value {
     dataInterface=&self:NiagaraDataInterfaceArrayFloat_0
     dataInterfaceClass=&NiagaraDataInterfaceArrayFloat
    }
   }
  }
  User.FlaresCount {type=NiagaraInt32, defaultValue=2}
  User.GameModeID {type=NiagaraInt32, defaultValue=3}
  User.PaintedColor {type=LinearColor, description="", defaultValue {struct=LinearColor, value=<rgba 1 1 1 1>}}
  User.PaintedScalar {type=NiagaraFloat, defaultValue=1, description=""}
  User.PortCount {type=NiagaraInt32, defaultValue=2}
  User.VehicleHasDriver {type=NiagaraBool, defaultValue=true}
 ]
}
