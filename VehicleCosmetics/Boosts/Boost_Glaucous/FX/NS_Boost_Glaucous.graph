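# Niagara system for the Glaucous vehicle boost cosmetic, trailing green and yellow exhaust from vehicle sockets.
# Five layered ribbon emitters and one sprite emitter use the filtered socket count, spawning a ribbon per port or partitioning sprites across sockets.
# Particles move along negative X while curves and dynamic material parameters shape width, size and fade.
# Painted-variant overrides are disabled, frontend-preview switches hold identical values, and SizeScaling affects only the first ribbon's width.
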NiagaraSystem /VehicleCosmetics/Boosts/Boost_Glaucous/FX/NS_Boost_Glaucous
dependencies {
 usedDataInterfaces [&NiagaraDataInterfaceVectorCurve, &NiagaraDataInterfaceCurve, &NiagaraDataInterfaceSkeletalMesh]
 usedDynamicInputs []
 usedModules [
  &self:SystemSpawnScript.SystemScriptSource.SystemScriptGraph.NiagaraNodeAssignment_0.SetVariables_#1
  &SystemState
 ]
}
emitters {
 NE_VC_Boost_Ribbon_Parent {
  emitterUpdateScript [
   &EmitterState(Inactive Response=::NewEnumerator1, Life Cycle Mode=::NewEnumerator1, Loop Duration=1, UseLoopDelay=false)
   &VC_FilteredSocketCount()
   &VC_SpawnRibbon_v1(
    Number of Ribbons=$Emitter.PortCount
    SpawnRate=&MakeCustomFloatFromBool(Bool=$User.bIsFrontendPreview, True Float=50, False Float=50)
   )
  ]
  particleSpawnScript [
   &VC_PaintedVariantsOverride() {enabled=false}
   &VC_InitializeRibbon_v1(
    Color=<rgba 0.0701 0.259 0.0701 1>
    Enable Skeletal Position=true
    Lifetime=1
    Ribbon Width:NiagaraFloat=&Multiply_Float(A=$User.SizeScaling, B=7.6)
   )
   &AddVelocity(
    Velocity=&MakeCustomVectorFromBool(Bool=true, True Vector=<1000 0 0>, False Vector=<1000 0 0>)
    Velocity Speed Scale=-1
   ) {module=AddVelocity001}
  ]
  particleUpdateScript [
   &ParticleState()
   &ScaleColor(
    Scale Alpha=&FloatFromCurve(FloatCurve=DataInterface() {propertyValues {Curve {keys [
         <at 0 5 RCIM_Cubic>
         {arriveTangent=-0.591, interpMode=RCIM_Cubic, leaveTangent=-0.591, tangentMode=RCTM_User, time=0.8, value=1}
         <at 1 0 RCIM_Cubic>
        ]}}})
    Scale RGB=&VectorFromCurve(VectorCurve=DataInterface() {propertyValues {XCurve {keys [<at 0 3 RCIM_Cubic>,<at 1 0>]}, YCurve {keys [<at 0 3 RCIM_Cubic>,<at 1 0>]}, ZCurve {keys [<at 0 3 RCIM_Cubic>,<at 1 0>]}}})
    ScaleA=true
    ScaleRGB=true
   )
   &DynamicMaterialParameters(
    Index 0 Param 1=1
    Index 0 Param 2=1
    Index 0 Param 3=0.2
    Index 0 Param 4=1.5
    Module.Index0Param0Write=true
    Module.Index0Param1Write=true
    Module.Index0Param2Write=true
    Module.Index0Param3Write=true
   )
   &ScaleRibbonWidth(Ribbon Width Scale=&FloatFromCurve(FloatCurve=DataInterface() {propertyValues {Curve {keys [<at 0 3 RCIM_Cubic>,<at 1 0>]}}}))
   &SolveForcesAndVelocity(Clamp Velocity=false, Limit Acceleration=false)
   &VC_UpdateRibbon_v1(Enable Skeletal Position=true)
  ]
  rendererClasses [&NiagaraRibbonRendererProperties]
  renderers [{rendererClass=&NiagaraRibbonRendererProperties}]
 }
 NE_VC_Boost_Ribbon_Parent001 {
  emitterUpdateScript [
   &EmitterState(Inactive Response=::NewEnumerator1, Life Cycle Mode=::NewEnumerator1, Loop Duration=1, UseLoopDelay=false)
   &VC_FilteredSocketCount()
   &VC_SpawnRibbon_v1(
    Number of Ribbons=$Emitter.PortCount
    SpawnRate=&MakeCustomFloatFromBool(Bool=$User.bIsFrontendPreview, True Float=120, False Float=120)
   )
  ]
  particleSpawnScript [
   &VC_PaintedVariantsOverride() {enabled=false}
   &VC_InitializeRibbon_v1(Color=<rgba 1.2 0.9709 0.0135 1>, Enable Skeletal Position=true, Lifetime=1, Ribbon Width=16)
   &AddVelocity(
    Velocity=&MakeCustomVectorFromBool(Bool=true, True Vector=<1000 0 0>, False Vector=<1000 0 0>)
    Velocity Speed Scale=-1
   ) {module=AddVelocity001}
  ]
  particleUpdateScript [
   &ParticleState()
   &ScaleColor(
    Scale Alpha=&FloatFromCurve(FloatCurve=DataInterface() {propertyValues {Curve {keys [
         <at 0 8 RCIM_Cubic>
         {arriveTangent=-0.591, interpMode=RCIM_Cubic, leaveTangent=-0.591, tangentMode=RCTM_User, time=0.8, value=2}
         <at 1 0 RCIM_Cubic>
        ]}}})
    Scale RGB=&VectorFromCurve(VectorCurve=DataInterface() {propertyValues {XCurve {keys [<at 0 1 RCIM_Cubic>,<at 1 0>]}, YCurve {keys [<at 0 1 RCIM_Cubic>,<at 1 0>]}, ZCurve {keys [<at 0 1 RCIM_Cubic>,<at 1 0>]}}})
    ScaleA=true
    ScaleRGB=true
   )
   &ScaleRibbonWidth(Ribbon Width Scale=&FloatFromCurve(FloatCurve=DataInterface() {propertyValues {Curve {keys [<at 0 1.5 RCIM_Cubic>,<at 1 0 RCIM_Cubic>]}}}))
   &DynamicMaterialParameters(
    Index 0 Param 1=1
    Index 0 Param 2=1.01
    Index 0 Param 3=0.05
    Index 0 Param 4=50
    Module.Index0Param0Write=true
    Module.Index0Param1Write=true
    Module.Index0Param2Write=true
    Module.Index0Param3Write=true
   )
   &SolveForcesAndVelocity(Clamp Velocity=false, Limit Acceleration=false)
   &VC_UpdateRibbon_v1(Enable Skeletal Position=true)
  ]
  rendererClasses [&NiagaraRibbonRendererProperties]
  renderers [{rendererClass=&NiagaraRibbonRendererProperties}]
 }
 NE_VC_Boost_Ribbon_Parent002 {
  emitterUpdateScript [
   &EmitterState(Life Cycle Mode=::NewEnumerator1, Loop Duration=1, UseLoopDelay=false)
   &VC_FilteredSocketCount()
   &VC_SpawnRibbon_v1(
    Number of Ribbons=$Emitter.PortCount
    SpawnRate=&MakeCustomFloatFromBool(Bool=$User.bIsFrontendPreview, True Float=120, False Float=120)
   )
  ]
  particleSpawnScript [
   &VC_PaintedVariantsOverride() {enabled=false}
   &VC_InitializeRibbon_v1(Color=<rgba 0.8 1 0.1 1>, Enable Skeletal Position=true, Lifetime=1, Ribbon Width=8)
   &AddVelocity(
    Velocity=&MakeCustomVectorFromBool(Bool=$User.bIsFrontendPreview, True Vector=<1000 0 0>, False Vector=<1000 0 0>)
    Velocity Speed Scale=-1
   ) {module=AddVelocity001}
  ]
  particleUpdateScript [
   &ParticleState()
   &ScaleColor(
    Scale Alpha=&FloatFromCurve(FloatCurve=DataInterface() {propertyValues {Curve {keys [
         <at 0 8 RCIM_Cubic>
         {arriveTangent=-0.591, interpMode=RCIM_Cubic, leaveTangent=-0.591, tangentMode=RCTM_User, time=0.35, value=4}
         <at 1 0 RCIM_Cubic>
        ]}}})
    Scale RGB=&VectorFromCurve(VectorCurve=DataInterface() {propertyValues {XCurve {keys [<at 0 1 RCIM_Cubic>,<at 1 0>]}, YCurve {keys [<at 0 1 RCIM_Cubic>,<at 1 0>]}, ZCurve {keys [<at 0 1 RCIM_Cubic>,<at 1 0>]}}})
    ScaleA=true
    ScaleRGB=true
   )
   &ScaleRibbonWidth(Ribbon Width Scale=&FloatFromCurve(FloatCurve=DataInterface() {propertyValues {Curve {keys [<at 0 3 RCIM_Cubic>,<at 0.8 0>]}}}))
   &DynamicMaterialParameters(
    Index 0 Param 1=0.1
    Index 0 Param 2=1
    Index 0 Param 4=5
    Module.Index0Param0Write=true
    Module.Index0Param1Write=true
    Module.Index0Param2Write=true
    Module.Index0Param3Write=true
   )
   &SolveForcesAndVelocity(Clamp Velocity=false, Limit Acceleration=false)
   &VC_UpdateRibbon_v1(Enable Skeletal Position=true)
  ]
  rendererClasses [&NiagaraRibbonRendererProperties]
  renderers [{rendererClass=&NiagaraRibbonRendererProperties}]
 }
 NE_VC_Boost_Ribbon_Parent003 {
  emitterUpdateScript [
   &EmitterState(Inactive Response=::NewEnumerator1, Life Cycle Mode=::NewEnumerator1, Loop Duration=1, UseLoopDelay=false)
   &VC_FilteredSocketCount()
   &VC_SpawnRibbon_v1(
    Number of Ribbons=$Emitter.PortCount
    SpawnRate=&MakeCustomFloatFromBool(Bool=$User.bIsFrontendPreview, True Float=120, False Float=120)
   )
  ]
  particleSpawnScript [
   &VC_PaintedVariantsOverride() {enabled=false}
   &VC_InitializeRibbon_v1(Color=<rgba 0.1 1 0.5 1>, Enable Skeletal Position=true, Lifetime=1, Ribbon Width=8)
   &AddVelocity(
    Velocity=&MakeCustomVectorFromBool(Bool=$User.bIsFrontendPreview, True Vector=<1000 0 0>, False Vector=<1000 0 0>)
    Velocity Speed Scale=-1
   ) {module=AddVelocity001}
  ]
  particleUpdateScript [
   &ParticleState()
   &ScaleColor(
    Scale Alpha=&FloatFromCurve(FloatCurve=DataInterface() {propertyValues {Curve {keys [
         <at 0 32 RCIM_Cubic>
         {arriveTangent=-0.591, interpMode=RCIM_Cubic, leaveTangent=-0.591, tangentMode=RCTM_User, time=0.35, value=4}
         <at 1 0 RCIM_Cubic>
        ]}}})
    Scale RGB=&VectorFromCurve(VectorCurve=DataInterface() {propertyValues {
       XCurve {keys [
         <at 0 1.5 RCIM_Cubic>
         {arriveTangent=-0.0204, interpMode=RCIM_Cubic, leaveTangent=-0.0204, tangentMode=RCTM_User, time=0.15, value=0.1}
         <at 0.6 0.01>
        ]}
       YCurve {keys [
         <at 0 1.5 RCIM_Cubic>
         {arriveTangent=-0.0204, interpMode=RCIM_Cubic, leaveTangent=-0.0204, tangentMode=RCTM_User, time=0.15, value=0.1}
         <at 0.6 0.01>
        ]}
       ZCurve {keys [
         <at 0 1.5 RCIM_Cubic>
         {arriveTangent=-0.0204, interpMode=RCIM_Cubic, leaveTangent=-0.0204, tangentMode=RCTM_User, time=0.15, value=0.1}
         <at 0.6 0.01>
        ]}
      }})
    ScaleA=true
    ScaleRGB=true
   )
   &DynamicMaterialParameters(
    Index 0 Param 1=3.0479
    Index 0 Param 2=1.0425
    Index 0 Param 3=0.05
    Index 0 Param 4=0.1
    Module.Index0Param0Write=true
    Module.Index0Param1Write=true
    Module.Index0Param2Write=true
    Module.Index0Param3Write=true
   )
   &ScaleRibbonWidth(Ribbon Width Scale=&FloatFromCurve(FloatCurve=DataInterface() {propertyValues {Curve {keys [<at 0 3 RCIM_Cubic>,<at 0.8 0>]}}}))
   &SolveForcesAndVelocity(Clamp Velocity=false, Limit Acceleration=false)
   &VC_UpdateRibbon_v1(Enable Skeletal Position=true)
  ]
  rendererClasses [&NiagaraRibbonRendererProperties]
  renderers [{rendererClass=&NiagaraRibbonRendererProperties}]
 }
 NE_VC_Boost_Ribbon_Parent004 {
  emitterUpdateScript [
   &EmitterState(Inactive Response=::NewEnumerator1, Life Cycle Mode=::NewEnumerator1, Loop Duration=1, UseLoopDelay=false)
   &VC_FilteredSocketCount()
   &VC_SpawnRibbon_v1(
    Number of Ribbons=$Emitter.PortCount
    SpawnRate=&MakeCustomFloatFromBool(Bool=$User.bIsFrontendPreview, True Float=30, False Float=30)
   )
  ]
  particleSpawnScript [
   &VC_PaintedVariantsOverride() {enabled=false}
   &VC_InitializeRibbon_v1(
    Color=<rgba 1.25 25 5 1>
    Enable Skeletal Position=true
    Lifetime=&RandomRangeFloat(Minimum=0.25, Maximum=0.65)
    Ribbon Width=18
   )
   &AddVelocity(
    Velocity=&MakeCustomVectorFromBool(Bool=true, True Vector=<1000 0 0>, False Vector=<1000 0 0>)
    Velocity Speed Scale=-1
   ) {module=AddVelocity001}
  ]
  particleUpdateScript [
   &ParticleState()
   &ScaleColor(
    Scale Alpha=&FloatFromCurve(FloatCurve=DataInterface() {propertyValues {Curve {keys [<at 0 1>,<at 0.3 1>,<at 1 1>]}}})
    Scale RGB=&VectorFromFloat(Value=1)
    ScaleA=true
    ScaleRGB=true
   )
   &DynamicMaterialParameters(
    Index 0 Param 1=2
    Index 0 Param 2=&RandomRangeFloat(Minimum=40, Maximum=80)
    Index 0 Param 3=2
    Index 0 Param 4=0.1
    Module.Index0Param0Write=true
    Module.Index0Param1Write=true
    Module.Index0Param2Write=false
    Module.Index0Param3Write=false
   )
   &ScaleRibbonWidth(Ribbon Width Scale=&FloatFromCurve(FloatCurve=DataInterface() {propertyValues {Curve {keys [<at 0 2 RCIM_Cubic>, {arriveTangent=-3.0769, interpMode=RCIM_Cubic, leaveTangent=-3.0769, time=0.4, value=1.5}, <at 0.65 0 RCIM_Cubic>]}}}))
   &SolveForcesAndVelocity(Clamp Velocity=false, Limit Acceleration=false)
   &VC_UpdateRibbon_v1(Enable Skeletal Position=true)
  ]
  rendererClasses [&NiagaraRibbonRendererProperties]
  renderers [{rendererClass=&NiagaraRibbonRendererProperties}]
 }
 NE_VC_Boost_Sprite_Parent {
  emitterUpdateScript [
   &EmitterState(Inactive Response=::NewEnumerator1, Life Cycle Mode=::NewEnumerator1, Loop Duration=1, UseLoopDelay=false)
   &VC_FilteredSocketCount()
   &SpawnRate(
    SpawnRate=&MakeCustomFloatFromBool(Bool=$User.bIsFrontendPreview, True Float=120, False Float=120)
    Use Spawn Probability=false
   )
  ]
  particleSpawnScript [
   &VC_PaintedVariantsOverride() {enabled=false}
   &InitializeParticle(
    Color=<rgba 5 100 40.0001 1>
    Lifetime=&RandomRangeFloat(Minimum=0.5, Maximum=0.75)
    Position Offset=<-75 0 0>
    Sprite Rotation Mode=::NewEnumerator2
    Sprite Size Mode=::NewEnumerator1
    Uniform Sprite Size Max=6
    Uniform Sprite Size Min=4
    UsePositionOffset=false
    Write Material Random=true
   )
   &PartitionParticles(Number Of Partitions=$Emitter.PortCount)
   &SkeletalMeshLocation(
    Bone / Socket Index=$Particles.PartitionParticles.Partition
    Bone / Socket Sampling Mode=::NewEnumerator5
    MaskSpawnGroup=false
    Mesh Sampling Type=::NewEnumerator6
    Skeletal Mesh=$System.SkeletalMesh
   )
   &ShapeLocation(MaskSpawnGroup=false, Sphere Distribution=::NewEnumerator1, Sphere Radius=5, UseOwnerScale=true)
   &AddVelocity(
    Velocity=&MakeCustomVectorFromBool(Bool=$User.bIsFrontendPreview, True Vector=<1000 0 0>, False Vector=<1000 0 0>)
    Velocity Speed Scale=-1
   )
  ]
  particleUpdateScript [
   &ParticleState()
   &ScaleSpriteSize(Uniform Curve Sprite Scale=DataInterface() {propertyValues {Curve {keys [<at 0 3 RCIM_Cubic>,<at 0.6 13 RCIM_Cubic>]}}})
   &ScaleColor(
    Scale Alpha=&FloatFromCurve(FloatCurve=DataInterface() {propertyValues {Curve {keys [<at 0 1 RCIM_Cubic>,<at 1 0 RCIM_Cubic>]}}})
    ScaleA=true
    ScaleRGB=true
   )
   &DynamicMaterialParameters(
    Index 0 Param 1=&RandomRangeFloat(Minimum=1, Maximum=2)
    Index 0 Param 2=&RandomRangeFloat(Minimum=0.2, Maximum=0.5)
    Index 0 Param 3=&RandomRangeFloat(Minimum=0.2, Maximum=0.35)
    Index 0 Param 4=&FloatFromCurve(FloatCurve=DataInterface() {propertyValues {Curve {keys [<at 0 0 RCIM_Cubic>,<at 1 0.5 RCIM_Cubic>]}}})
    Module.Index0Param0Write=true
    Module.Index0Param1Write=true
    Module.Index0Param2Write=true
    Module.Index0Param3Write=true
   )
   &SolveForcesAndVelocity(Clamp Velocity=false, Limit Acceleration=false)
  ]
  rendererClasses [&NiagaraSpriteRendererProperties]
  renderers [{rendererClass=&NiagaraSpriteRendererProperties}]
 }
}
summary {
 systemName=NS_Boost_Glaucous
 emitters [
  {emitterName=NE_VC_Boost_Ribbon_Parent, rendererClasses [&NiagaraRibbonRendererProperties]}
  {emitterName=NE_VC_Boost_Ribbon_Parent001, rendererClasses [&NiagaraRibbonRendererProperties]}
  {emitterName=NE_VC_Boost_Ribbon_Parent002, rendererClasses [&NiagaraRibbonRendererProperties]}
  {emitterName=NE_VC_Boost_Ribbon_Parent003, rendererClasses [&NiagaraRibbonRendererProperties]}
  {emitterName=NE_VC_Boost_Ribbon_Parent004, rendererClasses [&NiagaraRibbonRendererProperties]}
  {emitterName=NE_VC_Boost_Sprite_Parent, rendererClasses [&NiagaraSpriteRendererProperties]}
 ]
 userVariables [
  User.bIsFrontendPreview {type=NiagaraBool}
  User.bIsPainted {type=NiagaraBool}
  User.ColorVariantID {type=NiagaraInt32}
  User.PaintedColor {type=LinearColor, description="", defaultValue {struct=LinearColor, value=<rgba 1 1 1 1>}}
  User.PaintedScalar {type=NiagaraFloat, defaultValue=1, description=""}
  User.PortCount {type=NiagaraInt32, defaultValue=1}
  User.SizeScaling {type=NiagaraFloat, defaultValue=1, description=""}
 ]
}
