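# Unlit masked material for Niagara mesh particles, used by the burst of the Glaucous vehicle boost cosmetic.
# Panning flow noise distorts a caustics texture, which warps a flame-cloud mask and, with it, offsets smoke texture UVs, all scaled by dynamic parameters.
# Desaturated smoke multiplies particle color, lifted in threshold bands by an R-and-B-only brightness estimate, then Brightness, into MF_FX_EmissiveAdjustments.
# Smoke alpha times flame cloud forms a saturated mask.
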Material /VehicleCosmetics/Boosts/Boost_Glaucous/Materials/BoostMesh_Burst_MAT
graph {
 properties {
  bUsedWithNiagaraMeshParticles=true
  blendMode=BLEND_Masked
  pixelDepthOffsetMode=PDOM_Legacy
  shadingModel=MSM_Unlit
  usedShadingModels=Unlit
 }
 nodes [
  n1 = TextureCoordinate() {uTiling=0.8, vTiling=0.8}
  n2 = TextureSample(UVs=Panner(Coordinate=TextureCoordinate() {uTiling=0.5, vTiling=0.5}) {speedY=1}) {texture=/DealershipSharedLibrary/FX/Textures/Noise/T_VCVFX_Flow03_N}
  MaterialExpressionMeshEmitterDynamicParameter_0 = DynamicParameter() {paramNames [Param1,Brightness,UVDistort,Param4]}
  n3 = Add(
   A=TextureCoordinate() {uTiling=0.5, vTiling=0.5}
   B=Multiply(A=ComponentMask(None=@n2.RGB) {g=true, r=true}, B=@MaterialExpressionMeshEmitterDynamicParameter_0.Param1)
  )
  n4 = ComponentMask(None=TextureSample(UVs=Panner(Coordinate=@n3) {speedY=1.25}) {texture=/DealershipSharedLibrary/FX/Textures/Noise/T_VCVFX_CausticsMaskSharp02_RGB}.RGB) {g=true, r=true}
  n5 = TextureSample(UVs=Add(A=Multiply(A=@n4, B=@MaterialExpressionMeshEmitterDynamicParameter_0.Param4), B=TextureCoordinate())) {texture=/DealershipSharedLibrary/FX/Textures/Gradient_Masks/T_VCVFX_FlameCloud_02}
  n6 = Multiply(A=Multiply(A=@n4, B=@n5.R), B=@MaterialExpressionMeshEmitterDynamicParameter_0.UVDistort)
  n7 = TextureSample(UVs=Panner(Coordinate=Add(A=@n1, B=@n6), Time=Constant() {r=1}) {speedX=0.05, speedY=0.1}) {samplerType=SAMPLERTYPE_Normal, texture=/DealershipSharedLibrary/FX/Textures/Normal/T_VCVFX_SmokeMask02_N}
  MaterialExpressionConstantClamp_4 = Saturate(None=Desaturation(None=@n7.RGB))
  MaterialExpressionMeshEmitterVertexColor_0 = ParticleColor()
  n8 = Add(A=@MaterialExpressionMeshEmitterVertexColor_0.R, B=@MaterialExpressionMeshEmitterVertexColor_0.R)
  MaterialExpressionConstantClamp_5 = Saturate(None=Divide(A=Add(A=@n8, B=@MaterialExpressionMeshEmitterVertexColor_0.B), B=Constant() {r=3}))
  n9 = Constant() {r=0.5}
  n10 = Add(A=@MaterialExpressionMeshEmitterVertexColor_0.RGB, B=Constant() {r=0.05})
  n11 = Add(A=@MaterialExpressionMeshEmitterVertexColor_0.RGB, B=Constant() {r=0.1})
  n12 = Multiply(A=@n11, B=Constant() {r=1.25})
  n13 = If(
   A=@MaterialExpressionConstantClamp_5
   B=Constant() {r=0.16}
   "A > B"=@n12
   "A == B"=@n12
   "A < B"=@MaterialExpressionMeshEmitterVertexColor_0.RGB
  )
  n14 = Desaturation(
   None=If(
    A=@MaterialExpressionConstantClamp_5
    B=Constant() {r=0.785}
    "A > B"=@MaterialExpressionMeshEmitterVertexColor_0.RGB
    "A == B"=@MaterialExpressionMeshEmitterVertexColor_0.RGB
    "A < B"=If(A=@MaterialExpressionConstantClamp_5, B=@n9, "A > B"=@n10, "A == B"=@n11, "A < B"=@n13)
   )
   Fraction=Constant() {r=0.1}
  )
  MaterialFunctionCall(Emissive=Multiply(A=Multiply(A=@MaterialExpressionConstantClamp_4, B=@n14), B=@MaterialExpressionMeshEmitterDynamicParameter_0.Brightness)) {materialFunction=&MF_FX_EmissiveAdjustments}
  MaterialExpressionConstantClamp_6 = Saturate(None=Multiply(A=Multiply(A=@n7.A, B=@n5.R), B=Constant() {r=1.5}))
  Multiply(A=@MaterialExpressionMeshEmitterVertexColor_0.RGB, B=@MaterialExpressionConstantClamp_6)
 ]
}
