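# Unlit, two-sided translucent lightning arc material for Niagara ribbons in the Glaucous vehicle boost effect.
# Scrolling lightning texture layers are blended by animated masks and sine waves, with static switches for optional, alternate flicker.
# The pattern drives emissive, tinted by particle color and dynamic Brightness, and opacity, faded near the U start.
# Time-driven sine and cosine offsets along U, scaled by WPO parameters, wobble the ribbon.
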Material /VehicleCosmetics/Boosts/Boost_Glaucous/Materials/LightningArc_MAT
graph {
 properties {
  bUsedWithNiagaraRibbons=true
  blendMode=BLEND_Translucent
  pixelDepthOffsetMode=PDOM_Legacy
  shadingModel=MSM_Unlit
  twoSided=true
  usedShadingModels=Unlit
 }
 nodes [
  n1 = Multiply(A=OneMinus(None=Time()), B=ScalarParameter() {defaultValue=1, group=WPO, parameterName=Speed})
  n2 = ComponentMask(None=TextureCoordinate()) {r=true}
  n3 = ScalarParameter() {defaultValue=1, group=WPO, parameterName=tile}
  n4 = Constant() {r=0.75}
  MaterialExpressionConstantClamp_12 = Saturate(None=Sine(None=Time()) {period=0.75})
  n5 = Divide(A=@n2, B=LinearInterpolate(A=Multiply(A=@n3, B=@n4), B=@n3, Alpha=@MaterialExpressionConstantClamp_12))
  n6 = Cosine(None=Add(A=@n1, B=@n5))
  n7 = Sine(None=Add(A=@n5, B=@n1))
  n8 = ScalarParameter() {defaultValue=1, group=WPO, parameterName=Intensity}
  MaterialExpressionConstantClamp_13 = Saturate(None=Cosine(None=Time()))
  MaterialExpressionConstantClamp_14 = Saturate(None=Sine(None=Time()) {period=3})
  n9 = Constant() {r=0.5}
  n10 = Multiply(
   A=AppendVector(A=AppendVector(A=@n6, B=@n7), B=Constant())
   B=LinearInterpolate(
    A=OneMinus(None=@n8)
    B=@n8
    Alpha=LinearInterpolate(A=@MaterialExpressionConstantClamp_13, B=@MaterialExpressionConstantClamp_14, Alpha=@n9)
   )
  )
  n11 = ScalarParameter() {defaultValue=1, group=WPO, parameterName=Amount}
  ConstantBiasScale(None=Transform(None=Multiply(A=@n10, B=@n11)))
  n12 = TextureCoordinate() {uTiling=2, vTiling=0.25}
  n13 = Constant() {r=1}
  n14 = TextureSample(UVs=Panner(Coordinate=Panner(Coordinate=@n12, Time=@n13) {speedY=0.5}) {speedX=-3}) {texture=/VehicleCosmetics/Boosts/Boost_Glaucous/Textures/T_VCVFX_Boost_Glaucous_GPP_Lightning_RGB}
  n15 = TextureCoordinate() {coordinateIndex=1, uTiling=0, vTiling=0}
  n16 = DynamicParameter() {paramNames [Time,Brightness,EdgeGlow,Param4]}
  n17 = Multiply(A=@n16.Time, B=Time())
  n18 = TextureSample(UVs=Panner(Coordinate=@n15, Time=@n17) {speedX=0.95, speedY=1.45}) {texture=/VehicleCosmetics/Boosts/Boost_Glaucous/Textures/T_VCVFX_Boost_Glaucous_GPP_Lightning_RGB_FB_03}
  n19 = Constant() {r=0.4}
  MaterialExpressionConstantClamp_8 = Saturate(None=Multiply(A=Subtract(A=@n18.G, B=@n19), B=Constant() {r=5}))
  n20 = TextureCoordinate() {uTiling=3, vTiling=0.2}
  n21 = Constant() {r=1}
  n22 = TextureSample(UVs=Panner(Coordinate=Panner(Coordinate=@n20, Time=@n21) {speedY=0.77}) {speedX=-2.75}) {texture=&T_VCVFX_Boost_Glaucous_GPP_Lightning_RGB}
  n23 = Constant() {r=0.4}
  MaterialExpressionConstantClamp_9 = Saturate(None=Multiply(A=Subtract(A=@n18.B, B=@n23), B=Constant() {r=5}))
  n24 = TextureCoordinate() {uTiling=6, vTiling=0.2}
  n25 = Constant() {r=1}
  n26 = TextureSample(UVs=Panner(Coordinate=Panner(Coordinate=@n24, Time=@n25) {speedY=0.025}) {speedX=-3}) {texture=&T_VCVFX_Boost_Glaucous_GPP_Lightning_RGB}
  n27 = TextureCoordinate() {uTiling=3, vTiling=0.15}
  n28 = Constant() {r=1}
  n29 = TextureSample(UVs=Panner(Coordinate=Panner(Coordinate=@n27, Time=@n28) {speedY=0.26}) {speedX=-2.75}) {texture=&T_VCVFX_Boost_Glaucous_GPP_Lightning_RGB}
  n30 = TextureSample(UVs=Panner(Coordinate=@n15, Time=@n17) {speedX=1, speedY=1.5}) {texture=&T_VCVFX_Boost_Glaucous_GPP_Lightning_RGB_FB_03}
  n31 = Constant() {r=0.4}
  MaterialExpressionConstantClamp_6 = Saturate(None=Multiply(A=Subtract(A=@n30.G, B=@n31), B=Constant() {r=5}))
  n32 = Constant() {r=0.4}
  MaterialExpressionConstantClamp_7 = Saturate(None=Multiply(A=Subtract(A=@n30.B, B=@n32), B=Constant() {r=5}))
  n33 = Multiply(A=TextureCoordinate() {uTiling=0.25}, B=Constant2Vector() {g=1, r=2})
  n34 = Time()
  n35 = Add(A=ComponentMask(None=Frac(None=@n33)) {r=true}, B=Multiply(A=OneMinus(None=@n34), B=Constant() {r=3}))
  MaterialExpressionConstantClamp_5 = Saturate(None=ConstantBiasScale(None=Sine(None=@n35)) {bias=-0.1, scale=1})
  n36 = LinearInterpolate(
   A=Add(
    A=LinearInterpolate(A=@n14.R, B=@n14.R, Alpha=@MaterialExpressionConstantClamp_8)
    B=LinearInterpolate(A=@n22.G, B=@n22.G, Alpha=@MaterialExpressionConstantClamp_9)
   )
   B=Add(
    A=LinearInterpolate(A=@n26.R, B=@n29.G, Alpha=@MaterialExpressionConstantClamp_6)
    B=LinearInterpolate(A=@n26.B, B=@n29.R, Alpha=@MaterialExpressionConstantClamp_7)
   )
   Alpha=@MaterialExpressionConstantClamp_5
  )
  n37 = Constant() {r=0.4}
  MaterialExpressionConstantClamp_3 = Saturate(None=Multiply(A=Subtract(A=@n18.R, B=@n37), B=Constant() {r=5}))
  n38 = Constant() {r=0.4}
  MaterialExpressionConstantClamp_4 = Saturate(None=Multiply(A=Subtract(A=@n30.R, B=@n38), B=Constant() {r=5}))
  n39 = StaticSwitchParameter(
   True=StaticSwitchParameter(True=@MaterialExpressionConstantClamp_3, False=@MaterialExpressionConstantClamp_4) {parameterName=AlternateFlicker}
   False=Constant() {r=1}
  ) {parameterName=Flickering}
  MaterialExpressionConstantClamp_10 = Saturate(None=Multiply(A=@n36, B=@n39))
  MaterialExpressionConstantClamp_11 = Saturate(None=ConstantBiasScale(None=Sine(None=@n35) {period=0.5}) {bias=-0.1, scale=0.8})
  n40 = LinearInterpolate(A=Constant() {r=10}, B=@n16.Brightness, Alpha=@MaterialExpressionConstantClamp_11)
  MaterialExpressionVertexColor_0 = ParticleColor()
  n41 = Constant() {r=0.2}
  MaterialFunctionCall(Emissive=Multiply(A=Multiply(A=@MaterialExpressionConstantClamp_10, B=@n40), B=Desaturation(None=@MaterialExpressionVertexColor_0.RGB, Fraction=@n41))) {materialFunction=&MF_FX_EmissiveAdjustments}
  MaterialExpressionConstantClamp_2 = Saturate(None=Multiply(A=@n36, B=Constant() {r=20}))
  MaterialExpressionConstantClamp_0 = Saturate(None=Multiply(A=@n39, B=@MaterialExpressionConstantClamp_2))
  MaterialExpressionConstantClamp_1 = Saturate(None=ConstantBiasScale(None=@n2) {bias=-0.005, scale=10})
  Multiply(A=@MaterialExpressionVertexColor_0.A, B=Multiply(A=@MaterialExpressionConstantClamp_0, B=@MaterialExpressionConstantClamp_1))
 ]
}
