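# Additive, unlit, two-sided particle material for the lightning glow in the Glaucous vehicle boost cosmetic.
# It draws a soft procedural radial spot without textures, using centered UVs, an inverted self dot product and a power falloff.
# A Panner with constant time acts as a fixed centering offset, so the spot does not scroll.
# Particle color tints the spot before MF_FX_EmissiveAdjustments, and particle alpha scales a separate clamped mask.
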Material /VehicleCosmetics/Boosts/Boost_Glaucous/Materials/LightningGlow_MAT
graph {
 properties {blendMode=BLEND_Additive, pixelDepthOffsetMode=PDOM_Legacy, shadingModel=MSM_Unlit, twoSided=true, usedShadingModels=Unlit}
 nodes [
  n1 = Add(A=Panner(Coordinate=TextureCoordinate() {uTiling=3, vTiling=3}, Time=Constant() {r=1}) {speedX=1.5, speedY=1.5}, B=Constant() {r=-3})
  MaterialExpressionConstantClamp_1 = Saturate(None=OneMinus(None=DotProduct(A=@n1, B=@n1)))
  n2 = Multiply(A=Power(Base=@MaterialExpressionConstantClamp_1, Exp=Constant() {r=8}), B=Constant() {r=2})
  MaterialExpressionVertexColor_0 = ParticleColor()
  MaterialExpressionConstantClamp_0 = Saturate(None=Multiply(A=@n2, B=@MaterialExpressionVertexColor_0.A))
  MaterialFunctionCall(Emissive=Multiply(A=@n2, B=@MaterialExpressionVertexColor_0.RGB)) {materialFunction=&MF_FX_EmissiveAdjustments}
 ]
}
