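# Additive, unlit, two-sided Niagara sprite material that draws glittering wispy sparkles for the Glaucous vehicle boost.
# Flow noise distorts a cloud mask and two panning caustics layers, whose product forms the wisps.
# Multiplying noise offset by camera vector and caustics against fixed noise creates view-dependent sparkles, heavily amplified within the wisps.
# Particle color tints the emissive, while dynamic parameters control sparkle tiling, wisp speed, scale and falloff.
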Material /VehicleCosmetics/Boosts/Boost_Glaucous/Materials/Whispy_Sparkles_MAT
graph {
 properties {
  bUsedWithNiagaraSprites=true
  blendMode=BLEND_Additive
  pixelDepthOffsetMode=PDOM_Legacy
  shadingModel=MSM_Unlit
  twoSided=true
  usedShadingModels=Unlit
 }
 nodes [
  MaterialExpressionVertexColor_1 = ParticleColor()
  n1 = TextureCoordinate() {uTiling=0.5, vTiling=0.5}
  n2 = ComponentMask(None=TextureSample(UVs=Panner(Coordinate=@n1) {speedY=0.2}) {texture=/DealershipSharedLibrary/FX/Textures/Noise/T_VCVFX_Flow03_N}.RGB) {g=true, r=true}
  n3 = Add(A=TextureCoordinate(), B=Multiply(A=@n2, B=Constant() {r=0.5}))
  n4 = Constant() {r=1}
  n5 = DynamicParameter() {paramNames [tile,speed,ShapeTile,Param4]}
  n6 = Multiply(A=TextureCoordinate() {uTiling=4, vTiling=4}, B=@n5.tile)
  n7 = CameraVectorWS()
  n8 = Add(A=Multiply(A=TextureCoordinate() {uTiling=2, vTiling=2}, B=@n5.ShapeTile), B=Multiply(A=@n2, B=Constant() {r=0.2}))
  n9 = Multiply(A=@n5.speed, B=Time())
  n10 = Add(A=Multiply(A=TextureCoordinate(), B=@n5.ShapeTile), B=Multiply(A=@n2, B=Constant() {r=0.15}))
  n11 = Multiply(
   A=TextureSample(UVs=Panner(Coordinate=@n8, Time=@n9) {speedX=0.2, speedY=-0.1}) {texture=/DealershipSharedLibrary/FX/Textures/Noise/T_VCVFX_CausticsMaskSharp02_RGB}.R
   B=TextureSample(UVs=Panner(Coordinate=@n10, Time=@n9) {speedX=-0.05, speedY=-0.1}) {texture=&T_VCVFX_CausticsMaskSharp02_RGB}.B
  )
  n12 = ConstantBiasScale(None=@n11) {bias=-0.5, scale=2}
  n13 = Add(A=@n6, B=Add(A=ComponentMask(None=@n7) {g=true, r=true}, B=ComponentMask(None=@n12) {g=true, r=true}))
  MaterialExpressionConstantClamp_3 = Saturate(None=Multiply(
    A=TextureSample(UVs=@n13) {texture=/DealershipSharedLibrary/FX/Textures/Noise/T_VCVFX_NoiseGeneric01Pack_RGB}.B
    B=TextureSample(UVs=@n6) {texture=&T_VCVFX_NoiseGeneric01Pack_RGB}.B
   ))
  n14 = Constant() {r=512}
  n15 = Multiply(
   A=TextureSample(UVs=Panner(Coordinate=@n3, Time=@n4) {speedX=0.08}) {texture=/DealershipSharedLibrary/FX/Textures/Gradient_Masks/T_VCVFX_FireCloud02_A}.RGB
   B=Multiply(A=Multiply(A=@MaterialExpressionConstantClamp_3, B=@n14), B=Power(Base=@n11, Exp=@n5.Param4))
  )
  MaterialExpressionConstantClamp_2 = Saturate(None=Multiply(A=@MaterialExpressionVertexColor_1.A, B=@n15))
  n16 = Constant() {r=20}
  MaterialFunctionCall(Emissive=Multiply(A=@MaterialExpressionVertexColor_1.RGB, B=Multiply(A=@n16, B=@n15))) {materialFunction=&MF_FX_EmissiveAdjustments}
 ]
}
